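# Anim Blueprint for the CloakedIron backpack that poses its wings to match player movement.
# WingPosition state machine picks pose-asset poses from player flags, nesting parachute, dive and float states for skydiving.
# Acceleration-driven additive lean blendspaces layer on every pose except Targeting, whose lean is unconnected, and Crouching.
# Seven chained spring AnimDynamics nodes, weighted by DynamicsAlpha, then add wing-bone jiggle; driving variables are not declared or set here.
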AnimBlueprint /BRCosmetics/Accessories/FORT_Backpacks/Backpack_M_MED_CloakedIron/Meshes/M_MED_CloakedIron_Pack_AnimBP
target_skeleton=&/M_MED_CloakedIron_Pack_Skeleton
graphs [
 AnimGraph {nodes [
   AnimGraphNode_StateMachine_0 = WingPosition() {is=title AnimGraphNode_StateMachine}
   n1 = Get DynamicsAlpha() -> (DynamicsAlpha:float)
   n2 = Anim Dynamics - Bone: dyn_wing_index_01_l(ComponentPose=Local To Component(LocalPose=@AnimGraphNode_StateMachine_0), Alpha=@n1, ExternalForce="0,0,0") {
    angularDampingOverride=0.9999
    angularSpringConstant=8000
    bAngularSpring=true
    bOverrideAngularDamping=true
    bUseGravityOverride=true
    gravityScale=0.35
    simulationSpace=RootRelative
    alphaBoolBlend {blendOption=Linear}
    boundBone {boneName=dyn_wing_index_01_l}
    componentAppliedLinearAccClamp=<2000 2000 2000>
    componentLinearAccScale=<1 1 1>
    componentLinearVelScale=<1 1 1>
    gravityOverride {z=-980}
    physicsBodyDefinitions [{
      boundBone {boneName=dyn_wing_index_01_l}
      boxExtents=<20 30 30>
      constraintSetup {angularLimitsMax=<0 15 10>, angularLimitsMin=<0 -15 -10>, angularTarget=<1 0 0>}
      localJointOffset=<0.5 0 0>
     }]
    sphericalLimits [{limitRadius=0, drivingBone {boneName=None}, sphereLocalOffset=<0 0 0>}]
   }
   n3 = Anim Dynamics - Bone: dyn_wing_mid_01_l(ComponentPose=@n2, Alpha=@n1, ExternalForce="0,0,0") {
    angularDampingOverride=0.9999
    angularSpringConstant=8000
    bAngularSpring=true
    bOverrideAngularDamping=true
    bUseGravityOverride=true
    gravityScale=0.35
    simulationSpace=RootRelative
    alphaBoolBlend {blendOption=Linear}
    boundBone {boneName=dyn_wing_mid_01_l}
    componentAppliedLinearAccClamp=<2000 2000 2000>
    componentLinearAccScale=<3 3 3>
    componentLinearVelScale=<3 3 3>
    gravityOverride {z=-980}
    physicsBodyDefinitions [{
      boundBone {boneName=dyn_wing_mid_01_l}
      boxExtents=<20 30 30>
      constraintSetup {angularLimitsMax=<0 15 10>, angularLimitsMin=<0 -15 -10>, angularTarget=<1 0 0>}
      localJointOffset=<0.5 0 0>
     }]
   }
   n4 = Anim Dynamics - Bone: dyn_wing_base_02_l(ComponentPose=@n3, Alpha=@n1, ExternalForce="0,0,0") {
    angularDampingOverride=0.9999
    angularSpringConstant=8000
    bAngularSpring=true
    bOverrideAngularDamping=true
    bUseGravityOverride=true
    gravityScale=0.35
    simulationSpace=RootRelative
    alphaBoolBlend {blendOption=Linear}
    boundBone {boneName=dyn_wing_base_02_l}
    componentAppliedLinearAccClamp=<2000 2000 2000>
    componentLinearAccScale=<3 3 3>
    componentLinearVelScale=<3 3 3>
    gravityOverride {z=-980}
    physicsBodyDefinitions [{
      boundBone {boneName=dyn_wing_base_02_l}
      boxExtents=<20 30 30>
      constraintSetup {angularLimitsMax=<0 15 10>, angularLimitsMin=<0 -15 -10>, angularTarget=<1 0 0>}
      localJointOffset=<0.5 0 0>
     }]
   }
   n5 = Anim Dynamics - Bone: dyn_wing_index_01_r(ComponentPose=@n4, Alpha=@n1, ExternalForce="0,0,0") {
    angularDampingOverride=0.9999
    angularSpringConstant=8000
    bAngularSpring=true
    bOverrideAngularDamping=true
    bUseGravityOverride=true
    gravityScale=0.35
    simulationSpace=RootRelative
    alphaBoolBlend {blendOption=Linear}
    boundBone {boneName=dyn_wing_index_01_r}
    componentAppliedLinearAccClamp=<2000 2000 2000>
    componentLinearAccScale=<3 3 3>
    componentLinearVelScale=<3 3 3>
    gravityOverride {z=-980}
    physicsBodyDefinitions [{
      boundBone {boneName=dyn_wing_index_01_r}
      boxExtents=<20 30 30>
      constraintSetup {angularLimitsMax=<0 15 10>, angularLimitsMin=<0 -15 -10>, angularTarget=<1 0 0>}
      localJointOffset=<0.5 0 0>
     }]
   }
   n6 = Anim Dynamics - Bone: dyn_wing_base_02_r(ComponentPose=@n5, Alpha=@n1, ExternalForce="0,0,0") {
    angularDampingOverride=0.9999
    angularSpringConstant=8000
    bAngularSpring=true
    bOverrideAngularDamping=true
    bUseGravityOverride=true
    gravityScale=0.35
    simulationSpace=RootRelative
    alphaBoolBlend {blendOption=Linear}
    boundBone {boneName=dyn_wing_base_02_r}
    componentAppliedLinearAccClamp=<2000 2000 2000>
    componentLinearAccScale=<3 3 3>
    componentLinearVelScale=<3 3 3>
    gravityOverride {z=-980}
    physicsBodyDefinitions [{
      boundBone {boneName=dyn_wing_base_02_r}
      boxExtents=<20 30 30>
      constraintSetup {angularLimitsMax=<0 15 10>, angularLimitsMin=<0 -15 -10>, angularTarget=<1 0 0>}
      localJointOffset=<0.5 0 0>
     }]
   }
   n7 = Anim Dynamics - Bone: dyn_wing_mid_01_r(ComponentPose=@n6, Alpha=@n1, ExternalForce="0,0,0") {
    angularDampingOverride=0.9999
    angularSpringConstant=8000
    bAngularSpring=true
    bOverrideAngularDamping=true
    bUseGravityOverride=true
    gravityScale=0.35
    simulationSpace=RootRelative
    alphaBoolBlend {blendOption=Linear}
    boundBone {boneName=dyn_wing_mid_01_r}
    componentAppliedLinearAccClamp=<2000 2000 2000>
    componentLinearAccScale=<3 3 3>
    componentLinearVelScale=<3 3 3>
    gravityOverride {z=-980}
    physicsBodyDefinitions [{
      boundBone {boneName=dyn_wing_mid_01_r}
      boxExtents=<20 30 30>
      constraintSetup {angularLimitsMax=<0 15 10>, angularLimitsMin=<0 -15 -10>, angularTarget=<1 0 0>}
      localJointOffset=<0.5 0 0>
     }]
   }
   n8 = Anim Dynamics - Bone: dyn_wing_pinky_01_r(ComponentPose=@n7, Alpha=@n1, ExternalForce="0,0,0") {
    angularDampingOverride=0.9999
    angularSpringConstant=8000
    bAngularSpring=true
    bOverrideAngularDamping=true
    bUseGravityOverride=true
    gravityScale=0.35
    simulationSpace=RootRelative
    alphaBoolBlend {blendOption=Linear}
    boundBone {boneName=dyn_wing_pinky_01_r}
    componentAppliedLinearAccClamp=<2000 2000 2000>
    componentLinearAccScale=<3 3 3>
    componentLinearVelScale=<3 3 3>
    gravityOverride {z=-980}
    physicsBodyDefinitions [{
      boundBone {boneName=dyn_wing_pinky_01_r}
      boxExtents=<20 30 30>
      constraintSetup {angularLimitsMax=<0 15 10>, angularLimitsMin=<0 -15 -10>, angularTarget=<1 0 0>}
      localJointOffset=<0.5 0 0>
     }]
   }
   Output Pose(Result=Component To Local(ComponentPose=@n8)) {layerGroup=None}
  ]}
 WingPosition {
  entry=Base
  path=AnimGraph.AnimGraphNode_StateMachine_0.WingPosition
  ids=id
  nodes [
   AnimStateNode_0 = Base() {is=title AnimStateNode}
   AnimStateNode_1 = Targeting() {is=title AnimStateNode}
   AnimStateNode_2 = Jumping() {is=title AnimStateNode}
   AnimStateNode_3 = Falling() {is=title AnimStateNode}
   AnimStateNode_4 = SkydiveMode() {is=title AnimStateNode}
   AnimStateNode_5 = Crouching() {is=title AnimStateNode}
   AnimStateTransitionNode_15 = AnimStateTransitionNode()
   AnimStateTransitionNode_16 = AnimStateTransitionNode()
   AnimStateTransitionNode_17 = AnimStateTransitionNode()
   AnimStateTransitionNode_18 = AnimStateTransitionNode()
   AnimStateTransitionNode_19 = AnimStateTransitionNode()
   AnimStateTransitionNode_2 = AnimStateTransitionNode()
   AnimStateTransitionNode_20 = AnimStateTransitionNode()
   AnimStateTransitionNode_21 = AnimStateTransitionNode()
   AnimStateTransitionNode_22 = AnimStateTransitionNode()
   AnimStateTransitionNode_23 = AnimStateTransitionNode()
   AnimStateTransitionNode_24 = AnimStateTransitionNode()
   AnimStateTransitionNode_3 = AnimStateTransitionNode() {crossfadeDuration=0.1}
  ]
  links [
   {
    dst=Crouching
    rule=AnimGraph.AnimGraphNode_StateMachine_0.WingPosition.AnimStateTransitionNode_21.AnimationTransitionGraph_0
    src=Base
    via=AnimStateTransitionNode_21
   }
   {
    dst=Falling
    rule=AnimGraph.AnimGraphNode_StateMachine_0.WingPosition.AnimStateTransitionNode_17.AnimationTransitionGraph_0
    src=Base
    via=AnimStateTransitionNode_17
   }
   {
    dst=Jumping
    rule=AnimGraph.AnimGraphNode_StateMachine_0.WingPosition.AnimStateTransitionNode_2.AnimationTransitionGraph_0
    src=Base
    via=AnimStateTransitionNode_2
   }
   {
    dst=SkydiveMode
    rule=AnimGraph.AnimGraphNode_StateMachine_0.WingPosition.AnimStateTransitionNode_20.AnimationTransitionGraph_0
    src=Base
    via=AnimStateTransitionNode_20
   }
   {
    dst=Targeting
    rule=AnimGraph.AnimGraphNode_StateMachine_0.WingPosition.AnimStateTransitionNode_23.AnimationTransitionGraph_0
    src=Base
    via=AnimStateTransitionNode_23
   }
   {
    dst=Base
    rule=AnimGraph.AnimGraphNode_StateMachine_0.WingPosition.AnimStateTransitionNode_22.AnimationTransitionGraph_0
    src=Crouching
    via=AnimStateTransitionNode_22
   }
   {
    dst=Base
    rule=AnimGraph.AnimGraphNode_StateMachine_0.WingPosition.AnimStateTransitionNode_16.AnimationTransitionGraph_0
    src=Falling
    via=AnimStateTransitionNode_16
   }
   {
    dst=Base
    rule=AnimGraph.AnimGraphNode_StateMachine_0.WingPosition.AnimStateTransitionNode_3.AnimationTransitionGraph_0
    src=Jumping
    via=AnimStateTransitionNode_3
   }
   {
    dst=Falling
    rule=AnimGraph.AnimGraphNode_StateMachine_0.WingPosition.AnimStateTransitionNode_15.AnimationTransitionGraph_0
    src=Jumping
    via=AnimStateTransitionNode_15
   }
   {
    dst=SkydiveMode
    rule=AnimGraph.AnimGraphNode_StateMachine_0.WingPosition.AnimStateTransitionNode_18.AnimationTransitionGraph_0
    src=Jumping
    via=AnimStateTransitionNode_18
   }
   {
    dst=Base
    rule=AnimGraph.AnimGraphNode_StateMachine_0.WingPosition.AnimStateTransitionNode_19.AnimationTransitionGraph_0
    src=SkydiveMode
    via=AnimStateTransitionNode_19
   }
   {
    dst=Base
    rule=AnimGraph.AnimGraphNode_StateMachine_0.WingPosition.AnimStateTransitionNode_24.AnimationTransitionGraph_0
    src=Targeting
    via=AnimStateTransitionNode_24
   }
  ]
 }
 Base {
  path=AnimGraph.AnimGraphNode_StateMachine_0.WingPosition.AnimStateNode_0.Base
  nodes [
   n1 = Get PlayerForwardAcceleration() -> (PlayerForwardAcceleration:float)
   n2 = Get PlayerSidewaysAcceleration() -> (PlayerSidewaysAcceleration:float)
   Output Animation Pose(Result=Apply Additive(
     Base=Wing_Movement_Poses_PoseAsset : Base!(PoseWeight:float=1) -> (Pose:Pose Link Structure) {
      is=title AnimGraphNode_PoseByName
      poseAsset=/BRCosmetics/Accessories/FORT_Backpacks/Backpack_M_MED_CloakedIron/Animation/Wing_Movement_Poses_PoseAsset
      poseName=Base
     }
     Additive=Blendspace Player 'BS_Wing_Leans_Ground'(X=@n2, Y=@n1) {blendSpace=/BRCosmetics/Accessories/FORT_Backpacks/Backpack_M_MED_CloakedIron/Animation/BS_Wing_Leans_Ground}
    ) {alphaBoolBlend {blendOption=Linear}})
  ]
 }
 Targeting {
  path=AnimGraph.AnimGraphNode_StateMachine_0.WingPosition.AnimStateNode_1.Targeting
  nodes [
   Apply Additive(Additive=Blendspace Player 'BS_Wing_Leans_Ground'(X=Get PlayerSidewaysAcceleration(), Y=0) {blendSpace=&BS_Wing_Leans_Ground}) {alphaBoolBlend {blendOption=Linear}}
   n1 = Wing_Movement_Poses_PoseAsset : Targeting!(PoseWeight:float=1) -> (Pose:Pose Link Structure) {
    is=title AnimGraphNode_PoseByName
    poseAsset=&Wing_Movement_Poses_PoseAsset
    poseName=Targeting
   }
   Output Animation Pose(Result=@n1)
   Get PlayerForwardAcceleration()
  ]
 }
 Jumping {
  path=AnimGraph.AnimGraphNode_StateMachine_0.WingPosition.AnimStateNode_2.Jumping
  nodes [
   n1 = Get PlayerForwardAcceleration()
   n2 = Get PlayerSidewaysAcceleration()
   Output Animation Pose(Result=Apply Additive(
     Base=Wing_Movement_Poses_PoseAsset : Jump!(PoseWeight:float=1) -> (Pose:Pose Link Structure) {
      is=title AnimGraphNode_PoseByName
      poseAsset=&Wing_Movement_Poses_PoseAsset
      poseName=Jump
     }
     Additive=Blendspace Player 'BS_Wing_Leans_Ground'(X=@n2, Y=@n1) {blendSpace=&BS_Wing_Leans_Ground}
    ) {alphaBoolBlend {blendOption=Linear}})
  ]
 }
 Falling {
  path=AnimGraph.AnimGraphNode_StateMachine_0.WingPosition.AnimStateNode_3.Falling
  nodes [
   n1 = Get PlayerForwardAcceleration()
   n2 = Get PlayerSidewaysAcceleration()
   Output Animation Pose(Result=Apply Additive(
     Base=Wing_Movement_Poses_PoseAsset : Jump(PoseWeight=1) {poseAsset=&Wing_Movement_Poses_PoseAsset, poseName=Jump}
     Additive=Blendspace Player 'BS_Wing_Leans_Ground'(X=@n2, Y=@n1) {blendSpace=&BS_Wing_Leans_Ground}
    ) {alphaBoolBlend {blendOption=Linear}})
  ]
 }
 SkydiveMode {
  path=AnimGraph.AnimGraphNode_StateMachine_0.WingPosition.AnimStateNode_4.SkydiveMode
  nodes [
   AnimGraphNode_StateMachine_0 = SkyDivingState() {is=title AnimGraphNode_StateMachine}
   Output Animation Pose(Result=@AnimGraphNode_StateMachine_0)
  ]
 }
 SkyDivingState {
  entry=Parashuting
  path=AnimGraph.AnimGraphNode_StateMachine_0.WingPosition.AnimStateNode_4.SkydiveMode.AnimGraphNode_StateMachine_0.SkyDivingState
  ids=id
  nodes [
   AnimStateNode_0 = Parashuting() {is=title AnimStateNode}
   AnimStateNode_1 = Diving() {is=title AnimStateNode}
   AnimStateNode_2 = floating() {is=title AnimStateNode}
   AnimStateTransitionNode_0 = AnimStateTransitionNode()
   AnimStateTransitionNode_1 = AnimStateTransitionNode()
   AnimStateTransitionNode_3 = AnimStateTransitionNode()
   AnimStateTransitionNode_4 = AnimStateTransitionNode()
   AnimStateTransitionNode_5 = AnimStateTransitionNode()
   AnimStateTransitionNode_6 = AnimStateTransitionNode()
  ]
  links [
   {
    dst=Parashuting
    rule=AnimGraph.AnimGraphNode_StateMachine_0.WingPosition.AnimStateNode_4.SkydiveMode.AnimGraphNode_StateMachine_0.SkyDivingState.AnimStateTransitionNode_1.Transition
    src=Diving
    via=AnimStateTransitionNode_1
   }
   {
    dst=floating
    rule=AnimGraph.AnimGraphNode_StateMachine_0.WingPosition.AnimStateNode_4.SkydiveMode.AnimGraphNode_StateMachine_0.SkyDivingState.AnimStateTransitionNode_3.Transition
    src=Diving
    via=AnimStateTransitionNode_3
   }
   {
    dst=Diving
    rule=AnimGraph.AnimGraphNode_StateMachine_0.WingPosition.AnimStateNode_4.SkydiveMode.AnimGraphNode_StateMachine_0.SkyDivingState.AnimStateTransitionNode_0.Transition
    src=Parashuting
    via=AnimStateTransitionNode_0
   }
   {
    dst=floating
    rule=AnimGraph.AnimGraphNode_StateMachine_0.WingPosition.AnimStateNode_4.SkydiveMode.AnimGraphNode_StateMachine_0.SkyDivingState.AnimStateTransitionNode_6.Transition
    src=Parashuting
    via=AnimStateTransitionNode_6
   }
   {
    dst=Diving
    rule=AnimGraph.AnimGraphNode_StateMachine_0.WingPosition.AnimStateNode_4.SkydiveMode.AnimGraphNode_StateMachine_0.SkyDivingState.AnimStateTransitionNode_4.Transition
    src=floating
    via=AnimStateTransitionNode_4
   }
   {
    dst=Parashuting
    rule=AnimGraph.AnimGraphNode_StateMachine_0.WingPosition.AnimStateNode_4.SkydiveMode.AnimGraphNode_StateMachine_0.SkyDivingState.AnimStateTransitionNode_5.Transition
    src=floating
    via=AnimStateTransitionNode_5
   }
  ]
 }
 Parashuting {
  path=AnimGraph.AnimGraphNode_StateMachine_0.WingPosition.AnimStateNode_4.SkydiveMode.AnimGraphNode_StateMachine_0.SkyDivingState.AnimStateNode_0.Parashuting
  nodes [
   n1 = Get PlayerForwardAcceleration()
   n2 = Get PlayerSidewaysAcceleration()
   n3 = Wing_Movement_Poses_PoseAsset : Float!(PoseWeight:float=1) -> (Pose:Pose Link Structure) {
    is=title AnimGraphNode_PoseByName
    poseAsset=&Wing_Movement_Poses_PoseAsset
    poseName=Float
   }
   Output Animation Pose(Result=Apply Additive(
     Base=@n3
     Additive=Blendspace Player 'BS_Wing_Leans_Parashuting'(X=@n2, Y=@n1) {blendSpace=/BRCosmetics/Accessories/FORT_Backpacks/Backpack_M_MED_CloakedIron/Animation/BS_Wing_Leans_Parashuting}
    ) {alphaBoolBlend {blendOption=Linear}})
  ]
 }
 Diving {
  path=AnimGraph.AnimGraphNode_StateMachine_0.WingPosition.AnimStateNode_4.SkydiveMode.AnimGraphNode_StateMachine_0.SkyDivingState.AnimStateNode_1.Diving
  nodes [
   n1 = Get PlayerSidewaysAcceleration()
   n2 = Get PlayerForwardAcceleration()
   Output Animation Pose(Result=Apply Additive(
     Base=Wing_Movement_Poses_PoseAsset : Dive!(PoseWeight:float=1) -> (Pose:Pose Link Structure) {
      is=title AnimGraphNode_PoseByName
      poseAsset=&Wing_Movement_Poses_PoseAsset
      poseName=Dive
     }
     Additive=Blendspace Player 'BS_Wing_Leans_Air'(X=@n1, Y=@n2) {blendSpace=/BRCosmetics/Accessories/FORT_Backpacks/Backpack_M_MED_CloakedIron/Animation/BS_Wing_Leans_Air}
    ) {alphaBoolBlend {blendOption=Linear}})
  ]
 }
 floating {
  path=AnimGraph.AnimGraphNode_StateMachine_0.WingPosition.AnimStateNode_4.SkydiveMode.AnimGraphNode_StateMachine_0.SkyDivingState.AnimStateNode_2.floating
  nodes [
   n1 = Get PlayerForwardAcceleration()
   n2 = Get PlayerSidewaysAcceleration()
   n3 = Wing_Movement_Poses_PoseAsset : Float(PoseWeight=1) {poseAsset=&Wing_Movement_Poses_PoseAsset, poseName=Float}
   Output Animation Pose(Result=Apply Additive(
     Base=@n3
     Additive=Blendspace Player 'BS_Wing_Leans_Air'(X=@n2, Y=@n1) {blendSpace=&BS_Wing_Leans_Air}
    ) {alphaBoolBlend {blendOption=Linear}})
  ]
 }
 Transition {
  path=AnimGraph.AnimGraphNode_StateMachine_0.WingPosition.AnimStateNode_4.SkydiveMode.AnimGraphNode_StateMachine_0.SkyDivingState.AnimStateTransitionNode_0.Transition
  nodes [Result(bCanEnterTransition=NOT Boolean(A=Get bIsParachuteOpened()))]
 }
 Transition {
  path=AnimGraph.AnimGraphNode_StateMachine_0.WingPosition.AnimStateNode_4.SkydiveMode.AnimGraphNode_StateMachine_0.SkyDivingState.AnimStateTransitionNode_1.Transition
  nodes [Result(bCanEnterTransition=Get bIsParachuteOpened())]
 }
 Transition {
  path=AnimGraph.AnimGraphNode_StateMachine_0.WingPosition.AnimStateNode_4.SkydiveMode.AnimGraphNode_StateMachine_0.SkyDivingState.AnimStateTransitionNode_3.Transition
  nodes [Result(bCanEnterTransition=NOT Boolean(A=Get bIsSkydiveDiveMode()))]
 }
 Transition {
  path=AnimGraph.AnimGraphNode_StateMachine_0.WingPosition.AnimStateNode_4.SkydiveMode.AnimGraphNode_StateMachine_0.SkyDivingState.AnimStateTransitionNode_4.Transition
  nodes [Result(bCanEnterTransition=Get bIsSkydiveDiveMode())]
 }
 Transition {
  path=AnimGraph.AnimGraphNode_StateMachine_0.WingPosition.AnimStateNode_4.SkydiveMode.AnimGraphNode_StateMachine_0.SkyDivingState.AnimStateTransitionNode_5.Transition
  nodes [Result(bCanEnterTransition=Get bIsParachuteOpened())]
 }
 Transition {
  path=AnimGraph.AnimGraphNode_StateMachine_0.WingPosition.AnimStateNode_4.SkydiveMode.AnimGraphNode_StateMachine_0.SkyDivingState.AnimStateTransitionNode_6.Transition
  nodes [Result(bCanEnterTransition=NOT Boolean(A=Get bIsParachuteOpened()))]
 }
 Crouching {
  path=AnimGraph.AnimGraphNode_StateMachine_0.WingPosition.AnimStateNode_5.Crouching
  nodes [
   n1 = Wing_Movement_Poses_PoseAsset : Crouch!(PoseWeight:float=1) -> (Pose:Pose Link Structure) {
    is=title AnimGraphNode_PoseByName
    poseAsset=&Wing_Movement_Poses_PoseAsset
    poseName=Crouch
   }
   Output Animation Pose(Result=@n1)
  ]
 }
 AnimationTransitionGraph_0 {
  path=AnimGraph.AnimGraphNode_StateMachine_0.WingPosition.AnimStateTransitionNode_15.AnimationTransitionGraph_0
  nodes [Result(bCanEnterTransition=Get bIsFalling())]
 }
 AnimationTransitionGraph_0 {
  path=AnimGraph.AnimGraphNode_StateMachine_0.WingPosition.AnimStateTransitionNode_16.AnimationTransitionGraph_0
  nodes [Result(bCanEnterTransition=NOT Boolean(A=Get bIsFalling()))]
 }
 AnimationTransitionGraph_0 {
  path=AnimGraph.AnimGraphNode_StateMachine_0.WingPosition.AnimStateTransitionNode_17.AnimationTransitionGraph_0
  nodes [Result(bCanEnterTransition=Get bIsFalling())]
 }
 AnimationTransitionGraph_0 {
  path=AnimGraph.AnimGraphNode_StateMachine_0.WingPosition.AnimStateTransitionNode_18.AnimationTransitionGraph_0
  nodes [Result(bCanEnterTransition=Get bIsSkyDiving())]
 }
 AnimationTransitionGraph_0 {
  path=AnimGraph.AnimGraphNode_StateMachine_0.WingPosition.AnimStateTransitionNode_19.AnimationTransitionGraph_0
  nodes [Result(bCanEnterTransition=NOT Boolean(A=Get bIsSkyDiving()))]
 }
 AnimationTransitionGraph_0 {
  path=AnimGraph.AnimGraphNode_StateMachine_0.WingPosition.AnimStateTransitionNode_2.AnimationTransitionGraph_0
  nodes [Result(bCanEnterTransition=Get bIsJumping())]
 }
 AnimationTransitionGraph_0 {
  path=AnimGraph.AnimGraphNode_StateMachine_0.WingPosition.AnimStateTransitionNode_20.AnimationTransitionGraph_0
  nodes [Result(bCanEnterTransition=Get bIsSkyDiving())]
 }
 AnimationTransitionGraph_0 {
  path=AnimGraph.AnimGraphNode_StateMachine_0.WingPosition.AnimStateTransitionNode_21.AnimationTransitionGraph_0
  nodes [Result(bCanEnterTransition=Get bIsCrouching())]
 }
 AnimationTransitionGraph_0 {
  path=AnimGraph.AnimGraphNode_StateMachine_0.WingPosition.AnimStateTransitionNode_22.AnimationTransitionGraph_0
  nodes [Result(bCanEnterTransition=NOT Boolean(A=Get bIsCrouching()))]
 }
 AnimationTransitionGraph_0 {
  path=AnimGraph.AnimGraphNode_StateMachine_0.WingPosition.AnimStateTransitionNode_23.AnimationTransitionGraph_0
  nodes [Result(bCanEnterTransition=Get bIsTargeting())]
 }
 AnimationTransitionGraph_0 {
  path=AnimGraph.AnimGraphNode_StateMachine_0.WingPosition.AnimStateTransitionNode_24.AnimationTransitionGraph_0
  nodes [Result(bCanEnterTransition=NOT Boolean(A=Get bIsTargeting()))]
 }
 AnimationTransitionGraph_0 {
  path=AnimGraph.AnimGraphNode_StateMachine_0.WingPosition.AnimStateTransitionNode_3.AnimationTransitionGraph_0
  nodes [Result(bCanEnterTransition=NOT Boolean(A=Get bIsJumping()))]
 }
 EventGraph {links [], nodes []}
 OnInitialUpdate {nodes [
   OnInitialUpdate!() -> (then, Context:"Anim Update Context Structure (by ref)", Node:"Anim Node Reference Structure (by ref)") {is=title K2Node_FunctionEntry}
   > Return Node()
  ]}
]
variables []
